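# Physics asset for the ProdigyFire female medium body cosmetic, set up for secondary motion of three bottles, a necklace and left and right shoulder fabric pieces.
# Small base bodies lock to spine_01, spine_05 or the clavicles, and each anchors a dyn body with tight angular limits; bottles also lock twist.
# Larger capsules cover spine, clavicles, upper arms and right thigh; the thigh is unconstrained and reuses the spine_01 shape name.
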PhysicsAsset /BRCosmetics/Characters/Player/Female/Medium/Bodies/F_MED_ProdigyFire/Meshes/F_MED_ProdigyFire_Physics_Up
bodies {
 base_bottle_01 [{shapeType=Capsule, length=0.505, radius=0.505, shapeName=base_bottle_01_capsule, rotation=<rot 0 -90 0>}]
 base_bottle_02 [{shapeType=Capsule, length=0.505, radius=0.505, shapeName=base_bottle_02_capsule, rotation=<rot 0 -90 0>}]
 base_bottle_03 [{shapeType=Capsule, length=0.505, radius=0.505, shapeName=base_bottle_03_capsule, rotation=<rot 0 -90 0>}]
 base_necklace_01 [{shapeType=Capsule, length=0.505, radius=0.505, shapeName=base_necklace_01_capsule}]
 base_shoulderfabric_01_l [{shapeType=Capsule, length=0.505, radius=0.505, shapeName=base_shoulderfabric_01_l_capsule}]
 base_shoulderfabric_01_r [{shapeType=Capsule, length=0.505, radius=0.505, shapeName=base_shoulderfabric_01_r_capsule}]
 base_shoulderfabric_02_l [{shapeType=Capsule, length=0.505, radius=0.505, shapeName=base_shoulderfabric_02_l_capsule}]
 base_shoulderfabric_02_r [{shapeType=Capsule, length=0.505, radius=0.505, shapeName=base_shoulderfabric_02_r_capsule}]
 base_shoulderfabric_03_l [{shapeType=Capsule, length=0.505, radius=0.505, shapeName=base_shoulderfabric_03_l_capsule}]
 base_shoulderfabric_03_r [{shapeType=Capsule, length=0.505, radius=0.505, shapeName=base_shoulderfabric_03_r_capsule}]
 clavicle_l [{shapeType=Capsule, length=5, radius=4.5, shapeName=clavicle_l_capsule, center=<4.5 0 0>, rotation=<rot 90 0 0>}]
 clavicle_r [{shapeType=Capsule, length=5, radius=4.5, shapeName=clavicle_r_capsule, center=<-4.5 0 0>, rotation=<rot 90 0 0>}]
 dyn_bottle_01 [{shapeType=Capsule, length=12, radius=1.3, shapeName=dyn_bottle_01_capsule, center=<2.5 -0.8 0>, rotation=<rot 75 3 0>}]
 dyn_bottle_02 [{shapeType=Capsule, length=12, radius=1.3, shapeName=dyn_bottle_02_capsule, center=<2.5 0.3 0>, rotation=<rot 90 11.8618 8.8618>}]
 dyn_bottle_03 [{shapeType=Capsule, length=12, radius=1.3, shapeName=dyn_bottle_03_capsule, center=<2.5 0.3 0>, rotation=<rot 90 12.3863 9.3863>}]
 dyn_necklace_01 [{shapeType=Capsule, length=6, radius=4.2, shapeName=dyn_necklace_01_capsule, center=<6.1599 2 0>, rotation=<rot 90 0 0>}]
 dyn_shoulderfabric_01_l [{shapeType=Capsule, length=2, radius=1.5, shapeName=dyn_shoulderfabric_01_l_capsule, center=<1.5 0 0>, rotation=<rot 90 0 0>}]
 dyn_shoulderfabric_01_r [{shapeType=Capsule, length=2, radius=1.5, shapeName=dyn_shoulderfabric_01_r_capsule, center=<1.5 0 0>, rotation=<rot 90 0 0>}]
 dyn_shoulderfabric_02_l [{shapeType=Capsule, length=2, radius=1.5, shapeName=dyn_shoulderfabric_02_l_capsule, center=<1.5 0 0>, rotation=<rot 90 0 0>}]
 dyn_shoulderfabric_02_r [{shapeType=Capsule, length=2, radius=1.5, shapeName=dyn_shoulderfabric_02_r_capsule, center=<1.5 0 0>, rotation=<rot 90 0 0>}]
 dyn_shoulderfabric_03_l [{shapeType=Capsule, length=2, radius=1.5, shapeName=dyn_shoulderfabric_03_l_capsule, center=<1.5 0 0>, rotation=<rot 90 0 0>}]
 dyn_shoulderfabric_03_r [{shapeType=Capsule, length=2, radius=1.5, shapeName=dyn_shoulderfabric_03_r_capsule, center=<1.5 0 -0.6748>, rotation=<rot 90 0 0>}]
 spine_01 [{shapeType=Capsule, length=14, radius=7.5, shapeName=spine_01_capsule, center=<-5.4148 0 5.656>, rotation=<rot 73.1498 -0.0004 0>}]
 spine_05 [{shapeType=Capsule, length=22, radius=9, shapeName=spine_05_capsule, center=<-5.716 1 -0.0152>}]
 thigh_r [{
   shapeType=Capsule
   length=14
   radius=7.7
   shapeName=spine_01_capsule
   center=<12.0001 0.8791 -1.4572>
   rotation=<rot 88.9905 176.4173 179.9997>
  }]
 upperarm_l [{shapeType=Capsule, length=14, radius=4.5, shapeName=upperarm_l_capsule, center=<8 0 0>, rotation=<rot 90 0 0>}]
 upperarm_r [{shapeType=Capsule, length=14, radius=4, shapeName=upperarm_r_capsule, center=<-8 0 0>, rotation=<rot 90 -90.9147 -90.9147>}]
}
constraints [
 {
  bone1Name=upperarm_r
  bone2Name=clavicle_r
  swing1LimitDegrees=45
  swing1Motion=Locked
  swing2LimitDegrees=45
  swing2Motion=Locked
  twistLimitDegrees=45
  twistMotion=Locked
 }
 {
  bone1Name=upperarm_l
  bone2Name=clavicle_l
  swing1LimitDegrees=45
  swing1Motion=Locked
  swing2LimitDegrees=45
  swing2Motion=Locked
  twistLimitDegrees=45
  twistMotion=Locked
 }
 {bone1Name=dyn_shoulderfabric_01_l, bone2Name=base_shoulderfabric_01_l, swing1LimitDegrees=45, swing2LimitDegrees=30, twistLimitDegrees=5}
 {bone1Name=dyn_shoulderfabric_02_l, bone2Name=base_shoulderfabric_02_l, swing1LimitDegrees=45, swing2LimitDegrees=30, twistLimitDegrees=5}
 {bone1Name=dyn_shoulderfabric_03_l, bone2Name=base_shoulderfabric_03_l, swing1LimitDegrees=15, swing2LimitDegrees=10, twistLimitDegrees=5}
 {bone1Name=dyn_shoulderfabric_01_r, bone2Name=base_shoulderfabric_01_r, swing1LimitDegrees=35, swing2LimitDegrees=20, twistLimitDegrees=5}
 {bone1Name=dyn_shoulderfabric_02_r, bone2Name=base_shoulderfabric_02_r, swing1LimitDegrees=25, swing2LimitDegrees=15, twistLimitDegrees=5}
 {bone1Name=dyn_shoulderfabric_03_r, bone2Name=base_shoulderfabric_03_r, swing1LimitDegrees=15, swing2LimitDegrees=10, twistLimitDegrees=5}
 {
  bone1Name=base_shoulderfabric_01_l
  bone2Name=clavicle_l
  swing1LimitDegrees=45
  swing1Motion=Locked
  swing2LimitDegrees=45
  swing2Motion=Locked
  twistLimitDegrees=45
  twistMotion=Locked
 }
 {
  bone1Name=base_shoulderfabric_02_l
  bone2Name=clavicle_l
  swing1LimitDegrees=45
  swing1Motion=Locked
  swing2LimitDegrees=45
  swing2Motion=Locked
  twistLimitDegrees=45
  twistMotion=Locked
 }
 {
  bone1Name=base_shoulderfabric_03_l
  bone2Name=clavicle_l
  swing1LimitDegrees=45
  swing1Motion=Locked
  swing2LimitDegrees=45
  swing2Motion=Locked
  twistLimitDegrees=45
  twistMotion=Locked
 }
 {
  bone1Name=base_shoulderfabric_01_r
  bone2Name=clavicle_r
  swing1LimitDegrees=45
  swing1Motion=Locked
  swing2LimitDegrees=45
  swing2Motion=Locked
  twistLimitDegrees=45
  twistMotion=Locked
 }
 {
  bone1Name=base_shoulderfabric_02_r
  bone2Name=clavicle_r
  swing1LimitDegrees=45
  swing1Motion=Locked
  swing2LimitDegrees=45
  swing2Motion=Locked
  twistLimitDegrees=45
  twistMotion=Locked
 }
 {
  bone1Name=base_shoulderfabric_03_r
  bone2Name=clavicle_r
  swing1LimitDegrees=45
  swing1Motion=Locked
  swing2LimitDegrees=45
  swing2Motion=Locked
  twistLimitDegrees=45
  twistMotion=Locked
 }
 {
  bone1Name=base_bottle_01
  bone2Name=spine_01
  swing1LimitDegrees=45
  swing1Motion=Locked
  swing2LimitDegrees=45
  swing2Motion=Locked
  twistLimitDegrees=45
  twistMotion=Locked
 }
 {bone1Name=dyn_bottle_01, bone2Name=base_bottle_01, swing1LimitDegrees=10, swing2LimitDegrees=5, twistLimitDegrees=2, twistMotion=Locked}
 {
  bone1Name=base_bottle_02
  bone2Name=spine_01
  swing1LimitDegrees=45
  swing1Motion=Locked
  swing2LimitDegrees=45
  swing2Motion=Locked
  twistLimitDegrees=45
  twistMotion=Locked
 }
 {bone1Name=dyn_bottle_02, bone2Name=base_bottle_02, swing1LimitDegrees=10, swing2LimitDegrees=5, twistLimitDegrees=2, twistMotion=Locked}
 {
  bone1Name=base_bottle_03
  bone2Name=spine_01
  swing1LimitDegrees=45
  swing1Motion=Locked
  swing2LimitDegrees=45
  swing2Motion=Locked
  twistLimitDegrees=45
  twistMotion=Locked
 }
 {bone1Name=dyn_bottle_03, bone2Name=base_bottle_03, swing1LimitDegrees=10, swing2LimitDegrees=5, twistLimitDegrees=2, twistMotion=Locked}
 {
  bone1Name=spine_05
  bone2Name=spine_01
  swing1LimitDegrees=45
  swing1Motion=Free
  swing2LimitDegrees=45
  swing2Motion=Free
  twistLimitDegrees=45
  twistMotion=Free
 }
 {
  bone1Name=base_necklace_01
  bone2Name=spine_05
  swing1LimitDegrees=45
  swing1Motion=Locked
  swing2LimitDegrees=45
  swing2Motion=Locked
  twistLimitDegrees=45
  twistMotion=Locked
 }
 {bone1Name=dyn_necklace_01, bone2Name=base_necklace_01, swing1LimitDegrees=15, swing2LimitDegrees=5, twistLimitDegrees=5}
]
