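# Post-process style AnimBlueprint for the F_MED_ProdigyFire player body cosmetic, adding secondary motion and pose corrections to accessories.
# Copies the attached parent's pose through a Control Rig, then applies weighted constraints, additive bone offsets, RigidBody and AnimDynamics to ribbons, straps and fabric.
# EventGraph computes offset alphas from crouch, skydive, long-weapon aiming and spine/head rotation deltas, plus a gravity override.
# "Is Crouching and Aiming" actually combines crouching with a head-rotation flag.
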AnimBlueprint /BRCosmetics/Characters/Player/Female/Medium/Bodies/F_MED_ProdigyFire/Meshes/F_MED_ProdigyFire_AnimBP
target_skeleton=&/F_MED_ProdigyFire_Skeleton
graphs [
 AnimGraph {nodes [
   n1 = Control Rig(Source=Copy Pose From Mesh() {bUseAttachedParent=true}) {
    alphaBoolBlend {blendOption=Linear}
    controlRigAssetReference {blueprintRigClass=&/F_MED_ProdigyFire_ControlRig.F_MED_ProdigyFire_ControlRig_C}
   }
   n2 = "Constraint - base_ribbon (PPPPPPP)"!(
    ComponentPose=Local To Component(LocalPose=@n1)
    Alpha=1
    ConstraintWeights_0=0.3823
    ConstraintWeights_1:float=0.2462
    ConstraintWeights_2:float=0.1757
    ConstraintWeights_3:float=0.1192
    ConstraintWeights_4:float=0.075
    ConstraintWeights_5:float=0.0009
    ConstraintWeights_6:float=0.0007
   ) -> (Pose) {
    alphaBoolBlend {blendOption=Linear}
    boneToModify {boneName=base_ribbon}
    constraintSetup [
     {targetBone {boneName=spine_05}}
     {targetBone {boneName=neck_02}}
     {targetBone {boneName=neck_01}}
     {targetBone {boneName=base_ribbon}}
     {targetBone {boneName=spine_04}}
     {targetBone {boneName=dyn_ribbon_r}}
     {targetBone {boneName=dyn_ribbon_l}}
    ]
    constraintWeights [1,1,1,1,1,1,1]
   }
   n3 = "Constraint - base_back_broch_r (PPPPPPPPPP)"!(
    ComponentPose=@n2
    Alpha=1
    ConstraintWeights_0=0.3847
    ConstraintWeights_1:float=0.2483
    ConstraintWeights_2:float=0.2217
    ConstraintWeights_3:float=0.0677
    ConstraintWeights_4:float=0.0339
    ConstraintWeights_5:float=0.0218
    ConstraintWeights_6:float=0.0155
    ConstraintWeights_7:float=0.0059
    ConstraintWeights_8:float=0.0004
    ConstraintWeights_9:float=0.0002
   ) -> (Pose) {
    alphaBoolBlend {blendOption=Linear}
    boneToModify {boneName=base_back_broch_r}
    constraintSetup [
     {targetBone {boneName=clavicle_r}}
     {targetBone {boneName=spine_05}}
     {targetBone {boneName=spine_04}}
     {targetBone {boneName=deform_upperarm_bk_r}}
     {targetBone {boneName=upperarm_twist_01_r}}
     {targetBone {boneName=spine_02}}
     {targetBone {boneName=spine_03}}
     {targetBone {boneName=deform_upperarm_up_r}}
     {targetBone {boneName=deform_upperarm_fr_r}}
     {targetBone {boneName=deform_pec_r}}
    ]
    constraintWeights [1,1,1,1,1,1,1,1,1,1]
   }
   n4 = Get InCrouching() -> (InCrouching:double)
   n5 = "Transform (Modify) Bone - Bone: base_bottle_01"(ComponentPose=@n3, Alpha=@n4, Translation="1,-0.5,0", Rotation="0,18,0", Scale="1,1,1") {
    rotationMode=BMM_Additive
    rotationSpace=BCS_BoneSpace
    translationMode=BMM_Additive
    translationSpace=BCS_BoneSpace
    alphaBoolBlend {blendOption=Linear}
    alphaScaleBiasClamp {bInterpResult=true}
    boneToModify {boneName=base_bottle_01}
    rotation {yaw=18}
    translation=<1 -0.5>
   }
   n6 = "Transform (Modify) Bone - Bone: base_bottle_02"(ComponentPose=@n5, Alpha=@n4, Translation="1.3,-0.5,0", Rotation="0,18,0", Scale="1,1,1") {
    rotationMode=BMM_Additive
    rotationSpace=BCS_BoneSpace
    translationMode=BMM_Additive
    translationSpace=BCS_BoneSpace
    alphaBoolBlend {blendOption=Linear}
    alphaScaleBiasClamp {bInterpResult=true}
    boneToModify {boneName=base_bottle_02}
    rotation {yaw=18}
    translation=<1.3 -0.5>
   }
   n7 = "Transform (Modify) Bone - Bone: base_bottle_03"(ComponentPose=@n6, Alpha=@n4, Translation="1.3,0,0", Rotation="0,5,0", Scale="1,1,1") {
    rotationMode=BMM_Additive
    rotationSpace=BCS_BoneSpace
    translationMode=BMM_Additive
    translationSpace=BCS_BoneSpace
    alphaBoolBlend {blendOption=Linear}
    alphaScaleBiasClamp {bInterpResult=true}
    boneToModify {boneName=base_bottle_03}
    rotation {yaw=5}
    translation {x=1.3}
   }
   n8 = "Transform (Modify) Bone - Bone: attach_cape"(ComponentPose=@n7, Alpha=@n4, Translation="0,1,-5.5", Rotation="0,0,0", Scale="1,1,1") {
    rotationMode=BMM_Additive
    rotationSpace=BCS_BoneSpace
    translationMode=BMM_Additive
    translationSpace=BCS_BoneSpace
    alphaBoolBlend {blendOption=Linear}
    alphaScaleBiasClamp {bInterpResult=true}
    boneToModify {boneName=attach_cape}
    translation {y=1, z=-5.5}
   }
   n9 = "Transform (Modify) Bone - Bone: dyn_back_broch_r"(ComponentPose=@n8, Alpha=Get HeadBack() -> (HeadBack:double), Translation="0,-0.2,0", Rotation="-0.008,-0.007,20.000", Scale="1,1,1") {
    rotationMode=BMM_Additive
    rotationSpace=BCS_BoneSpace
    translationMode=BMM_Additive
    translationSpace=BCS_BoneSpace
    alphaBoolBlend {blendOption=Linear}
    alphaScaleBiasClamp {bInterpResult=true}
    boneToModify {boneName=dyn_back_broch_r}
    rotation=<rot -0.0076 -0.0069 19.9995>
    translation {y=-0.2}
   }
   n10 = Get Is Crouching and Aiming() -> (Is Crouching and Aiming:double)
   n11 = "Transform (Modify) Bone - Bone: base_shoulderfabric_03_r"(ComponentPose=@n9, Alpha=@n10, Translation="1.33,0.5,-1.285", Rotation="-10,75,-40", Scale="1,1,1") {
    rotationMode=BMM_Additive
    rotationSpace=BCS_BoneSpace
    translationMode=BMM_Additive
    translationSpace=BCS_BoneSpace
    alphaBoolBlend {blendOption=Linear}
    alphaScaleBiasClamp {bInterpResult=true}
    boneToModify {boneName=base_shoulderfabric_03_r}
    rotation=<rot -10 75 -40>
    translation=<1.33 0.5 -1.285>
   }
   n12 = "Transform (Modify) Bone - Bone: base_shoulderfabric_04_r"(ComponentPose=@n11, Alpha=@n10, Translation="0.797,1,-0.3", Rotation="20,45,0", Scale="1,1,1") {
    rotationMode=BMM_Additive
    rotationSpace=BCS_BoneSpace
    translationMode=BMM_Additive
    translationSpace=BCS_BoneSpace
    alphaBoolBlend {blendOption=Linear}
    alphaScaleBiasClamp {bInterpResult=true}
    boneToModify {boneName=base_shoulderfabric_04_r}
    rotation {pitch=20, yaw=45}
    translation=<0.797 1 -0.3>
   }
   n13 = "Transform (Modify) Bone - Bone: attach_cape"(ComponentPose=@n12, Alpha=@n10, Translation="0,2,-1", Rotation="0,0,0", Scale="1,1,1") {
    rotationMode=BMM_Additive
    rotationSpace=BCS_BoneSpace
    translationMode=BMM_Additive
    translationSpace=BCS_BoneSpace
    alphaBoolBlend {blendOption=Linear}
    alphaScaleBiasClamp {bInterpResult=true}
    boneToModify {boneName=attach_cape}
    translation {y=2, z=-1}
   }
   n14 = Get isSkyDiving() -> (isSkyDiving:double)
   n15 = "Transform (Modify) Bone - Bone: dyn_back_broch_r"(ComponentPose=@n13, Alpha=@n14, Translation="0,0,0", Rotation="0,10,-15", Scale="1,1,1") {
    rotationMode=BMM_Additive
    rotationSpace=BCS_BoneSpace
    translationMode=BMM_Additive
    translationSpace=BCS_BoneSpace
    alphaBoolBlend {blendOption=Linear}
    alphaScaleBiasClamp {bInterpResult=true}
    boneToModify {boneName=dyn_back_broch_r}
    rotation {roll=-15, yaw=10}
   }
   n16 = "Transform (Modify) Bone - Bone: attach_cape"(ComponentPose=@n15, Alpha=@n14, Translation="0,1.1,-6", Rotation="0,0,0", Scale="1,1,1") {
    rotationMode=BMM_Additive
    rotationSpace=BCS_BoneSpace
    translationMode=BMM_Additive
    translationSpace=BCS_BoneSpace
    alphaBoolBlend {blendOption=Linear}
    alphaScaleBiasClamp {bInterpResult=true}
    boneToModify {boneName=attach_cape}
    translation {y=1.1, z=-6}
   }
   n17 = RigidBody!(ComponentPose=@n16, Alpha, OverrideWorldGravity:Vector=Get GravityOverride(), ExternalForce) -> (Pose) {
    bClampLinearTranslationLimitToRefPose=true
    bOverrideWorldGravity=true
    overridePhysicsAsset=&/F_MED_ProdigyFire_Physics_Up
    simulationSpace=BaseBoneSpace
    alphaBoolBlend {blendOption=Linear}
    baseBoneRef {boneName=spine_05}
    componentAppliedLinearAccClamp=<3000 3000 3000>
    componentLinearAccScale=<0.6 0.6 0.6>
    componentLinearVelScale=<0.6 0.6 0.6>
    simSpaceSettings {
     maxAngularAcceleration=1000
     maxAngularVelocity=3
     maxLinearAcceleration=700
     maxLinearVelocity=500
     velocityScaleZ=0
     worldAlpha=0.2
     externalLinearDragV=<0.2 0.2>
    }
   }
   n18 = "Transform (Modify) Bone - Bone: base_necklace_01"(ComponentPose=@n17, Alpha=Get HatBackRight() -> (HatBackRight:double), Translation="0,1,0", Rotation="0,-10,0", Scale="1,1,1") {
    rotationMode=BMM_Additive
    rotationSpace=BCS_BoneSpace
    translationMode=BMM_Additive
    translationSpace=BCS_BoneSpace
    alphaBoolBlend {blendOption=Linear}
    alphaScaleBiasClamp {bInterpResult=true}
    boneToModify {boneName=base_necklace_01}
    rotation {yaw=-10}
    translation {y=1}
   }
   n19 = "Transform (Modify) Bone - Bone: base_necklace_01"(
    ComponentPose=@n18
    Alpha=Get Is Aiming and LongWeapon() -> (Is Aiming and LongWeapon:double)
    Translation="0,3,0"
    Rotation="0,-20,0"
    Scale="1,1,1"
   ) {
    rotationMode=BMM_Additive
    rotationSpace=BCS_BoneSpace
    translationMode=BMM_Additive
    translationSpace=BCS_BoneSpace
    alphaBoolBlend {blendOption=Linear}
    alphaScaleBiasClamp {bInterpResult=true}
    boneToModify {boneName=base_necklace_01}
    rotation {yaw=-20}
    translation {y=3}
   }
   n20 = Anim Dynamics - Bone: dyn_ribbon_r(ComponentPose=@n19, Alpha=1, ExternalForce="0,0,0") {
    angularDampingOverride=1
    angularSpringConstant=20000
    bAngularSpring=true
    bOverrideAngularDamping=true
    bUseGravityOverride=true
    gravityScale=0.35
    simulationSpace=RootRelative
    alphaBoolBlend {blendOption=Linear}
    boundBone {boneName=dyn_ribbon_r}
    componentLinearAccScale=<1 1 1>
    componentLinearVelScale=<1 1 1>
    gravityOverride {z=-980}
    physicsBodyDefinitions [{
      boundBone {boneName=dyn_ribbon_r}
      boxExtents=<20 30 30>
      constraintSetup {angularLimitsMax=<2 10 10>, angularLimitsMin=<-2 -10 -10>, angularTarget=<1 0 0>}
      localJointOffset=<4 0 0>
     }]
   }
   n21 = Anim Dynamics - Bone: dyn_ribbon_l(ComponentPose=@n20, Alpha=1, ExternalForce="0,0,0") {
    angularDampingOverride=1
    angularSpringConstant=20000
    bAngularSpring=true
    bOverrideAngularDamping=true
    bUseGravityOverride=true
    gravityScale=0.35
    simulationSpace=RootRelative
    alphaBoolBlend {blendOption=Linear}
    boundBone {boneName=dyn_ribbon_l}
    componentLinearAccScale=<1 1 1>
    componentLinearVelScale=<1 1 1>
    gravityOverride {z=-980}
    physicsBodyDefinitions [{
      boundBone {boneName=dyn_ribbon_l}
      boxExtents=<20 30 30>
      constraintSetup {angularLimitsMax=<2 10 10>, angularLimitsMin=<-2 -10 -10>, angularTarget=<1 0 0>}
      localJointOffset=<4 0 0>
     }]
   }
   n22 = Anim Dynamics - Bone: dyn_bootbelt_r(ComponentPose=@n21, Alpha=1, ExternalForce="0,0,0") {
    angularDampingOverride=1
    angularSpringConstant=20000
    bAngularSpring=true
    bOverrideAngularDamping=true
    bUseGravityOverride=true
    gravityScale=0.35
    simulationSpace=RootRelative
    alphaBoolBlend {blendOption=Linear}
    boundBone {boneName=dyn_bootbelt_r}
    componentLinearAccScale=<1 1 1>
    componentLinearVelScale=<1 1 1>
    gravityOverride {z=-980}
    physicsBodyDefinitions [{
      boundBone {boneName=dyn_bootbelt_r}
      boxExtents=<20 30 30>
      constraintSetup {angularLimitsMax=<15 35 15>, angularLimitsMin=<-15 -35 -15>, angularTarget=<1 0 0>}
      localJointOffset=<4 0 0>
     }]
   }
   n23 = Anim Dynamics - Bone: dyn_bootbelt_l(ComponentPose=@n22, Alpha=1, ExternalForce="0,0,0") {
    angularDampingOverride=1
    angularSpringConstant=20000
    bAngularSpring=true
    bOverrideAngularDamping=true
    bUseGravityOverride=true
    gravityScale=0.35
    simulationSpace=RootRelative
    alphaBoolBlend {blendOption=Linear}
    boundBone {boneName=dyn_bootbelt_l}
    componentLinearAccScale=<1 1 1>
    componentLinearVelScale=<1 1 1>
    gravityOverride {z=-980}
    physicsBodyDefinitions [{
      boundBone {boneName=dyn_bootbelt_l}
      boxExtents=<20 30 30>
      constraintSetup {angularLimitsMax=<15 35 15>, angularLimitsMin=<-15 -35 -15>, angularTarget=<1 0 0>}
      localJointOffset=<4 0 0>
     }]
   }
   n24 = Anim Dynamics - Bone: dyn_shoulderfabric_04_l(ComponentPose=@n23, Alpha=1, ExternalForce="0,0,0") {
    angularDampingOverride=1
    angularSpringConstant=20000
    bAngularSpring=true
    bOverrideAngularDamping=true
    bUseGravityOverride=true
    gravityScale=0.35
    simulationSpace=RootRelative
    alphaBoolBlend {blendOption=Linear}
    boundBone {boneName=dyn_shoulderfabric_04_l}
    componentLinearAccScale=<1 1 1>
    componentLinearVelScale=<1 1 1>
    gravityOverride {z=-980}
    physicsBodyDefinitions [{
      boundBone {boneName=dyn_shoulderfabric_04_l}
      boxExtents=<20 30 30>
      constraintSetup {angularLimitsMax=<2.5 5 5>, angularLimitsMin=<-2.5 -5 -10>, angularTarget=<1 0 0>}
      localJointOffset=<0 0 0>
     }]
   }
   n25 = Anim Dynamics - Bone: dyn_shoulderfabric_05_l(ComponentPose=@n24, Alpha=1, ExternalForce="0,0,0") {
    angularDampingOverride=1
    angularSpringConstant=20000
    bAngularSpring=true
    bOverrideAngularDamping=true
    bUseGravityOverride=true
    gravityScale=0.35
    simulationSpace=RootRelative
    alphaBoolBlend {blendOption=Linear}
    boundBone {boneName=dyn_shoulderfabric_05_l}
    componentLinearAccScale=<1 1 1>
    componentLinearVelScale=<1 1 1>
    gravityOverride {z=-980}
    physicsBodyDefinitions [{
      boundBone {boneName=dyn_shoulderfabric_05_l}
      boxExtents=<20 30 30>
      constraintSetup {angularLimitsMax=<2.5 5 5>, angularLimitsMin=<-2.5 -5 -10>, angularTarget=<1 0 0>}
      localJointOffset=<0 0 0>
     }]
   }
   n26 = Anim Dynamics - Bone: dyn_shoulderfabric_04_r(ComponentPose=@n25, Alpha=1, ExternalForce="0,0,0") {
    angularDampingOverride=1
    angularSpringConstant=20000
    bAngularSpring=true
    bOverrideAngularDamping=true
    bUseGravityOverride=true
    gravityScale=0.35
    simulationSpace=RootRelative
    alphaBoolBlend {blendOption=Linear}
    boundBone {boneName=dyn_shoulderfabric_04_r}
    componentLinearAccScale=<1 1 1>
    componentLinearVelScale=<1 1 1>
    gravityOverride {z=-980}
    physicsBodyDefinitions [{
      boundBone {boneName=dyn_shoulderfabric_04_r}
      boxExtents=<20 30 30>
      constraintSetup {angularLimitsMax=<2.5 5 5>, angularLimitsMin=<-2.5 -5 -10>, angularTarget=<1 0 0>}
      localJointOffset=<0 0 0>
     }]
   }
   n27 = Anim Dynamics - Bone: dyn_shoulderfabric_05_r(ComponentPose=@n26, Alpha=1, ExternalForce="0,0,0") {
    angularDampingOverride=1
    angularSpringConstant=20000
    bAngularSpring=true
    bOverrideAngularDamping=true
    bUseGravityOverride=true
    gravityScale=0.35
    simulationSpace=RootRelative
    alphaBoolBlend {blendOption=Linear}
    boundBone {boneName=dyn_shoulderfabric_05_r}
    componentLinearAccScale=<1 1 1>
    componentLinearVelScale=<1 1 1>
    gravityOverride {z=-980}
    physicsBodyDefinitions [{
      boundBone {boneName=dyn_shoulderfabric_05_r}
      boxExtents=<20 30 30>
      constraintSetup {angularLimitsMax=<2.5 5 5>, angularLimitsMin=<-2.5 -5 -10>, angularTarget=<1 0 0>}
      localJointOffset=<0 0 0>
     }]
   }
   Output Pose(Result=Component To Local(ComponentPose=@n27)) {layerGroup=None}
  ]}
 EventGraph {nodes [
   Event BlueprintUpdateAnimation()
   > n1 = Sequence!(execute) -> (then_0, then_1, then_2:Exec, then_3:Exec, then_4:Exec, then_5:Exec, then_6:Exec, then_7:Exec)
   Set InCrouching(execute=@n1.then_2, InCrouching:double="To Float (Boolean)"(InBool=Get bIsCrouching())) -> (Output_Get:double)
   n2 = Cast To FortPlayerAnimInstance(execute=@n1.then_7, Object=GetAnimInstance(self=GetOwnerSkeletalMeshComponent()))
   > n3 = Switch on EFortWeaponCoreAnimation!(execute, Selection:EFortWeaponCoreAnimation Enum=Get WeaponCoreAnim(self=@n2)) -> (
    Melee:Exec
    Pistol:Exec
    Shotgun:Exec
    PaperBlueprint:Exec
    Rifle:Exec
    MeleeOneHand:Exec
    MachinePistol:Exec
    RocketLauncher:Exec
    GrenadeLauncher:Exec
    GoingCommando:Exec
    AssaultRifle:Exec
    TacticalShotgun:Exec
    SniperRifle:Exec
    TrapPlacement:Exec
    ShoulderLauncher:Exec
    AbilityDecoTool:Exec
    Crossbow:Exec
    C4:Exec
    RemoteControl:Exec
    DualWield:Exec
    AR_BullPup:Exec
    AR_ForwardGrip:Exec
    MedPackPaddles:Exec
    SMG_P90:Exec
    AR_DrumGun:Exec
    Consumable_Small:Exec
    Consumable_Large:Exec
    Balloon:Exec
    MountedTurret:Exec
    CreativeTool:Exec
    ExplosiveBow:Exec
    AshtonIndigo:Exec
    AshtonChicago:Exec
    MeleeDualWield:Exec
    Unarmed:Exec
   )
   n4 = Branch(
    execute=[@n3.AssaultRifle, @n3.GoingCommando, @n3.GrenadeLauncher, @n3.Rifle, @n3.Shotgun, @n3.SniperRifle, @n3.TacticalShotgun]
    Condition=Get bIsTargeting()
   )
   > Set LongWeaponEquip(LongWeaponEquip:bool=true) -> (Output_Get:bool)
   Set LongWeaponEquip(execute=@n4.else, LongWeaponEquip=false)
   n5 = AND Boolean(A=Get LongWeaponEquip() -> (LongWeaponEquip:bool), B=Get AimingDown() -> (AimingDown:bool))
   n6 = "Not Equal (Boolean)"(A=Get bIsCrouching(), B=true) -> (ReturnValue=~)
   Set Is Aiming and LongWeapon(execute=@n1.then_4, Is Aiming and LongWeapon:double="To Float (Boolean)"(InBool=AND Boolean(A=@n5, B=@n6))) -> (Output_Get:double)
   n7 = Get Skel Mesh()
   Set AimingDown(
    execute=@n1.then_6
    AimingDown:bool="float > float"(
     A=GetAlphaFromReferenceTransformDelta(SkeletalMeshComponent=@n7, BoneName=spine_05, BaseName=spine_01, InputRange=(X=0,Y=90), TwistAxis="0,0,1")
     B=0.18
    )
   ) -> (Output_Get:bool)
   n8 = Get bIsCrouching()
   n9 = Get Is Hat Back Right() -> (Is Hat Back Right:bool)
   Set Is Crouching and Aiming(execute=@n1.then_3, Is Crouching and Aiming:double="To Float (Boolean)"(InBool=AND Boolean(A=@n9, B=@n8))) -> (Output_Get:double)
   Set HeadSideL(
    execute=@n1.then_5
    HeadSideL:double=GetAlphaFromReferenceTransformDelta(SkeletalMeshComponent=@n7, BoneName=spine_05, BaseName=head, DeltaComponentType=EulerY, InputRange=(X=0,Y=-60), TwistAxis="0,0,1")
   ) -> (Output_Get:double)
   > n10 = Set HeadBack(HeadBack:double=GetAlphaFromReferenceTransformDelta(SkeletalMeshComponent=@n7, BoneName=spine_05, BaseName=head, InputRange=(X=0,Y=90), TwistAxis="0,0,1")) -> (Output_Get:double)
   > n11 = Set HatBack(HatBack:double=@n10) -> (Output_Get:double)
   n12 = float * float(A=Get HeadBack(), B=0.2) {is=title K2Node_PromotableOperator}
   n13 = Get HeadSideL() -> (HeadSideL:double)
   > Set HatBackRight(HatBackRight:double=float * float(A=@n11, B=float + float(A=@n13, B=@n12))) -> (Output_Get:double)
   n14 = float * float(A=@n10, B=2)
   > Set Is Hat Back Right(Is Hat Back Right:bool="To Boolean (Integer)"!(InInt:int=Truncate(A=@n14)) -> (ReturnValue:bool)) -> (Output_Get:bool)
   n15 = Get bIsSkydiveDiveMode()
   n16 = Get Is Sky Floating() -> (Is Sky Floating:bool)
   n17 = Get bIsSkydiveFloating()
   Set isSkyDiving(execute=@n1.then_1, isSkyDiving:double="To Float (Boolean)"(InBool=OR Boolean!(A=@n15, B=@n17, C:bool=@n16) -> (ReturnValue))) -> (Output_Get:double)
   n18 = Gravity Override State Function DEPRECATED(
    execute=@n1.then_0
    CharacterPartAnimInstance=Self-Reference()
    GameTimeInSeconds=GetGameTimeInSeconds()
    GravityOverrideParams=Get Gravity Override Params()
   )
   > Set GravityOverride(GravityOverride=@n18)
   Event BlueprintInitializeAnimation()
   > Set Skel Mesh(Skel Mesh=GetOwnerSkeletalMeshComponent())
  ]}
]
variables [
 AimingDown {type=boolean}
 ar_firingPosition {type=integer}
 Body_Mesh {type=SkeletalMeshComponent}
 Gravity Override Params {type=GravityOverrideParamsStruct}
 GravityOverride {type=Vector}
 HatBack {type=number}
 HatBackRight {type=number}
 HeadBack {type=number}
 HeadSideL {type=number}
 InCrouching {type=number}
 Is Aiming and LongWeapon {type=number}
 Is Crouching and Aiming {type=number}
 Is Floating And Backpack {type=number}
 Is Hat Back Right {type=boolean}
 Is Sky Floating {type=boolean}
 Is Wearing Cape {type=boolean}
 isSkyDiving {type=number}
 isWearingBackbling {type=boolean}
 LongWeaponEquip {type=boolean}
 Neck Rot {type=number}
 Skel Mesh {type=SkeletalMeshComponent}
]
