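# Physics asset for the ProdigyFire female character's skydive front-hair face accessory, covering hair, hat brim and earrings.
# Capsule bodies cover head, neck, clavicles and base/dyn bone pairs, with several base bones locked to the head.
# Dyn bones form short, narrowly limited chains from their bases, some hair roots branching into multiple tips.
# Brim roots and left back-hair root are locked, while dyn_hair_a_04_r rotates freely from the head.
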PhysicsAsset /BRCosmetics/Characters/Player/Female/Medium/Bodies/F_MED_ProdigyFire/Meshes/Parts/F_MED_ProdigyFire_FaceAcc_FrontHairSkyDive_Physics
bodies {
 base_brimhat_01_back_r [{shapeType=Capsule, length=0.505, radius=0.505, shapeName=base_brimhat_01_back_r_capsule, rotation=<rot 0 -90 0>}]
 base_brimhat_01_r [{shapeType=Capsule, length=0.505, radius=0.505, shapeName=base_brimhat_01_r_capsule}]
 base_earring_l [{shapeType=Capsule, length=0.505, radius=0.505, shapeName=base_earring_l_capsule}]
 base_earring_r [{shapeType=Capsule, length=0.505, radius=0.505, shapeName=base_earring_r_capsule}]
 base_hair_01_back_l [{shapeType=Capsule, length=0.505, radius=0.505, shapeName=base_hair_01_back_l_capsule}]
 base_hair_01_back_m [{shapeType=Capsule, length=0.505, radius=0.505, shapeName=base_hair_01_front_l_capsule}]
 base_hair_01_back_m_l [{shapeType=Capsule, length=0.505, radius=0.505, shapeName=base_hair_01_front_l_capsule}]
 base_hair_01_back_m_r [{shapeType=Capsule, length=0.505, radius=0.505, shapeName=base_hair_01_front_l_capsule}]
 base_hair_01_back_r [{shapeType=Capsule, length=0.505, radius=0.505, shapeName=base_hair_01_back_r_capsule}]
 clavicle_l [{shapeType=Capsule, length=9, radius=7.5, shapeName=clavicle_l_capsule, center=<5.7043 0 -0.9159>, rotation=<rot 90 0 0>}]
 clavicle_r [{shapeType=Capsule, length=9, radius=7.5, shapeName=clavicle_l_capsule, center=<-5.7043 0 0.9159>, rotation=<rot 90 0 0>}]
 dyn_brimhat_01_back_r [{shapeType=Capsule, length=1.5, radius=1.5, shapeName=dyn_brimhat_02_r_capsule, center=<1.5 0.5 0>, rotation=<rot 90 0 0>}]
 dyn_brimhat_01_r [{shapeType=Capsule, length=0.5, radius=1.5, shapeName=dyn_brimhat_02_r_capsule, center=<0.8 0 0>, rotation=<rot 90 0 0>}]
 dyn_brimhat_02_back_r [{shapeType=Capsule, length=5, radius=1.6, shapeName=dyn_brimhat_02_r_capsule, center=<3 0.5 0>, rotation=<rot 90 0 0>}]
 dyn_brimhat_02_r [{shapeType=Capsule, length=0.5, radius=1.5, shapeName=dyn_brimhat_02_r_capsule, center=<0.8 0 0>, rotation=<rot 90 0 0>}]
 dyn_earring_l [{shapeType=Capsule, length=0.1, radius=3, shapeName=dyn_earring_l_capsule, center=<2.7621 0 -1.7191>}]
 dyn_earring_r [{shapeType=Capsule, length=0.1, radius=3, shapeName=dyn_earring_r_capsule, center=<2.7621 0 1.5882>}]
 dyn_hair_01_back_l [{shapeType=Capsule, length=1.5, radius=2, shapeName=dyn_hair_01_back_l_capsule, center=<1.3 0 0>, rotation=<rot 90 0 0>}]
 dyn_hair_01_back_m [{shapeType=Capsule, length=2, radius=1, shapeName=dyn_hair_01_back_m_capsule, center=<1 0 0>, rotation=<rot 90 0 0>}]
 dyn_hair_01_back_m_l [{shapeType=Capsule, length=2, radius=1, shapeName=dyn_hair_01_back_m_l_capsule, center=<1 0 0>, rotation=<rot 90 0 0>}]
 dyn_hair_01_back_m_r [{shapeType=Capsule, length=2, radius=1, shapeName=dyn_hair_01_back_m_r_capsule, center=<1 0 0>, rotation=<rot 90 0 0>}]
 dyn_hair_01_back_r [{shapeType=Capsule, length=2, radius=1, shapeName=dyn_earring_r_capsule, center=<1 0 0>, rotation=<rot 90 0 0>}]
 dyn_hair_02_back_l [{
   shapeType=Capsule
   length=6
   radius=1.3
   shapeName=dyn_hair_02_back_l_capsule
   center=<3.7294 0.8303 0.4937>
   rotation=<rot 80.0001 20.0015 0.0002>
  }]
 dyn_hair_02_back_m [{shapeType=Capsule, length=4, radius=1.5, shapeName=dyn_hair_02_back_m_capsule, center=<2.5 0 0>, rotation=<rot 80 180 180>}]
 dyn_hair_02_back_m_l [{shapeType=Capsule, length=5, radius=1, shapeName=dyn_hair_02_back_m_l_capsule, center=<2.5 0 -0.4222>, rotation=<rot 70 0 0>}]
 dyn_hair_02_back_m_lft [{shapeType=Capsule, length=3, radius=1.5, shapeName=dyn_hair_02_back_m_lft_capsule, center=<1.8 0 0>, rotation=<rot 90 0 0>}]
 dyn_hair_02_back_m_r [{shapeType=Capsule, length=2.2, radius=1.2, shapeName=dyn_hair_02_back_m_r_capsule, center=<1.2 -0.302 0>, rotation=<rot 90 180 -160.0001>}]
 dyn_hair_02_back_m_r_down [{shapeType=Capsule, length=2.5, radius=1.5, shapeName=dyn_hair_02_back_m_r_down_capsule, center=<1.7 0 0>, rotation=<rot 90 0 0>}]
 dyn_hair_02_back_m_rght [{shapeType=Capsule, length=3.5, radius=1.6, shapeName=dyn_hair_02_back_m_rght_capsule, center=<2.5 0 0>, rotation=<rot 90 0 0>}]
 dyn_hair_02_back_r1 [{
   shapeType=Capsule
   length=4
   radius=0.7
   shapeName=dyn_hair_02_back_r1_capsule
   center=<0.0386 0.4486 -0.0771>
   rotation=<rot 86.5523 82.0767 96.6633>
  }]
 dyn_hair_02_back_r2 [{shapeType=Capsule, length=6, radius=0.6, shapeName=dyn_hair_02_back_r2_capsule, center=<0.9997 -0.1471 -0.0875>, rotation=<rot 90 0 0>}]
 dyn_hair_02_back_r3 [{
   shapeType=Capsule
   length=6.3
   radius=1.3
   shapeName=dyn_hair_02_back_r3_capsule
   center=<0.1776 0.8096 0.3252>
   rotation=<rot 0 -89.269 -78.0832>
  }]
 dyn_hair_02_back_r4 [{
   shapeType=Capsule
   length=4.5
   radius=1.8
   shapeName=dyn_hair_02_back_r4_capsule
   center=<3.3542 -0.6897 0.1592>
   rotation=<rot 82.2454 -40.9975 -55.1078>
  }]
 dyn_hair_a_04_r [{shapeType=Capsule, length=4, radius=1.2, shapeName=dyn_hair_a_04_r_capsule, center=<3 0 0>, rotation=<rot 90 0 0>}]
 head [{shapeType=Capsule, length=6, radius=7, shapeName=head_capsule, center=<1.5044 1.2624 0>, rotation=<rot 90 0 -40>}]
 neck_01 [{shapeType=Capsule, length=4, radius=4.5, shapeName=neck_01_capsule, center=<0 0.6175 0>, rotation=<rot 90 0 0>}]
}
constraints [
 {
  bone1Name=base_brimhat_01_r
  bone2Name=head
  swing1LimitDegrees=45
  swing1Motion=Locked
  swing2LimitDegrees=45
  swing2Motion=Locked
  twistLimitDegrees=45
  twistMotion=Locked
 }
 {
  bone1Name=base_earring_l
  bone2Name=head
  swing1LimitDegrees=45
  swing1Motion=Locked
  swing2LimitDegrees=45
  swing2Motion=Locked
  twistLimitDegrees=45
  twistMotion=Locked
 }
 {
  bone1Name=base_earring_r
  bone2Name=head
  swing1LimitDegrees=45
  swing1Motion=Locked
  swing2LimitDegrees=45
  swing2Motion=Locked
  twistLimitDegrees=45
  twistMotion=Locked
 }
 {
  bone1Name=base_hair_01_back_l
  bone2Name=head
  swing1LimitDegrees=45
  swing1Motion=Locked
  swing2LimitDegrees=45
  swing2Motion=Locked
  twistLimitDegrees=45
  twistMotion=Locked
 }
 {
  bone1Name=base_hair_01_back_r
  bone2Name=head
  swing1LimitDegrees=45
  swing1Motion=Locked
  swing2LimitDegrees=45
  swing2Motion=Locked
  twistLimitDegrees=45
  twistMotion=Locked
 }
 {bone1Name=dyn_hair_01_back_r, bone2Name=base_hair_01_back_r, swing1LimitDegrees=10, swing2LimitDegrees=10, twistLimitDegrees=5}
 {bone1Name=dyn_hair_02_back_l, bone2Name=dyn_hair_01_back_l, swing1LimitDegrees=10, swing2LimitDegrees=10, twistLimitDegrees=5}
 {
  bone1Name=dyn_hair_01_back_l
  bone2Name=base_hair_01_back_l
  swing1LimitDegrees=10
  swing1Motion=Locked
  swing2LimitDegrees=3
  swing2Motion=Locked
  twistLimitDegrees=5
  twistMotion=Locked
 }
 {bone1Name=dyn_earring_r, bone2Name=base_earring_r, swing1LimitDegrees=15, swing2LimitDegrees=20, twistLimitDegrees=10}
 {bone1Name=dyn_earring_l, bone2Name=base_earring_l, swing1LimitDegrees=4, swing2LimitDegrees=8, twistLimitDegrees=3}
 {bone1Name=dyn_brimhat_02_r, bone2Name=dyn_brimhat_01_r, swing1LimitDegrees=5, swing2LimitDegrees=2, twistLimitDegrees=1}
 {
  bone1Name=dyn_brimhat_01_r
  bone2Name=base_brimhat_01_r
  swing1LimitDegrees=5
  swing1Motion=Locked
  swing2LimitDegrees=2
  swing2Motion=Locked
  twistLimitDegrees=2
  twistMotion=Locked
 }
 {
  bone1Name=head
  bone2Name=neck_01
  swing1LimitDegrees=45
  swing1Motion=Locked
  swing2LimitDegrees=45
  swing2Motion=Locked
  twistLimitDegrees=45
  twistMotion=Locked
 }
 {bone1Name=dyn_hair_02_back_r1, bone2Name=dyn_hair_01_back_r, swing1LimitDegrees=10, swing2LimitDegrees=10, twistLimitDegrees=5}
 {bone1Name=dyn_hair_02_back_r2, bone2Name=dyn_hair_01_back_r, swing1LimitDegrees=10, swing2LimitDegrees=10, twistLimitDegrees=5}
 {bone1Name=dyn_hair_02_back_r3, bone2Name=dyn_hair_01_back_r, swing1LimitDegrees=10, swing2LimitDegrees=10, twistLimitDegrees=5}
 {bone1Name=dyn_hair_02_back_r4, bone2Name=dyn_hair_01_back_r, swing1LimitDegrees=10, swing2LimitDegrees=10, twistLimitDegrees=5}
 {bone1Name=dyn_hair_01_back_m, bone2Name=base_hair_01_back_m, swing1LimitDegrees=10, swing2LimitDegrees=10, twistLimitDegrees=5}
 {bone1Name=dyn_hair_02_back_m, bone2Name=dyn_hair_01_back_m, swing1LimitDegrees=15, swing2LimitDegrees=10, twistLimitDegrees=5}
 {bone1Name=dyn_hair_02_back_m_lft, bone2Name=dyn_hair_01_back_m, swing1LimitDegrees=15, swing2LimitDegrees=10, twistLimitDegrees=5}
 {bone1Name=dyn_hair_02_back_m_rght, bone2Name=dyn_hair_01_back_m, swing1LimitDegrees=15, swing2LimitDegrees=15, twistLimitDegrees=5}
 {bone1Name=dyn_hair_01_back_m_l, bone2Name=base_hair_01_back_m_l, swing1LimitDegrees=10, swing2LimitDegrees=10, twistLimitDegrees=5}
 {bone1Name=dyn_hair_02_back_m_l, bone2Name=dyn_hair_01_back_m_l, swing1LimitDegrees=10, swing2LimitDegrees=10, twistLimitDegrees=5}
 {bone1Name=dyn_hair_01_back_m_r, bone2Name=base_hair_01_back_m_r, swing1LimitDegrees=18, swing2LimitDegrees=10, twistLimitDegrees=5}
 {bone1Name=dyn_hair_02_back_m_r, bone2Name=dyn_hair_01_back_m_r, swing1LimitDegrees=30, swing2LimitDegrees=25, twistLimitDegrees=10}
 {bone1Name=dyn_hair_02_back_m_r_down, bone2Name=dyn_hair_01_back_m_r, swing1LimitDegrees=30, swing2LimitDegrees=20, twistLimitDegrees=10}
 {
  bone1Name=dyn_brimhat_01_back_r
  bone2Name=base_brimhat_01_back_r
  swing1LimitDegrees=20
  swing1Motion=Locked
  swing2LimitDegrees=2
  swing2Motion=Locked
  twistLimitDegrees=2
  twistMotion=Locked
 }
 {bone1Name=dyn_brimhat_02_back_r, bone2Name=dyn_brimhat_01_back_r, swing1LimitDegrees=5, swing2LimitDegrees=2, twistLimitDegrees=1}
 {
  bone1Name=dyn_hair_a_04_r
  bone2Name=head
  swing1LimitDegrees=45
  swing1Motion=Free
  swing2LimitDegrees=45
  swing2Motion=Free
  twistLimitDegrees=45
  twistMotion=Free
 }
]
