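# Anim Blueprint for ProdigyFire's face accessory part, animating hair and hat brim with physics and correcting earring placement.
# It copies parent and body-mesh poses into separate normal and skydiving chains, each using bone offsets, RigidBody and AnimDynamics.
# Head-tilt alphas drive normal-state corrections; backpack or cape flags adjust the brim while skydiving.
# Output switches on skydive or parachute state, then layers DefaultSlot from the front-hair bone by a front-end idle curve.
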AnimBlueprint /BRCosmetics/Characters/Player/Female/Medium/Bodies/F_MED_ProdigyFire/Meshes/Parts/F_MED_ProdigyFire_FaceAcc_AnimBP
target_skeleton=&/F_MED_ProdigyFire_FaceAcc_Skeleton
graphs [
 AnimGraph {nodes [
   n1 = Use cached pose 'Cache_SkydivingBackpackCapeState'()
   n2 = Use cached pose 'Cache_NormalState'()
   n3 = Get isSkyDiving() -> (isSkyDiving:bool)
   Output Pose(Result=Layered blend per bone(
     BasePose=Blend Poses by bool(BlendPose_0=@n1, BlendPose_1=@n2, bActiveValue=@n3) {blendType=Linear}
     BlendPoses_0=Slot 'DefaultSlot'()
     BlendWeights_0=Get Custom Idle Is Playing Float() -> (Custom Idle Is Playing Float:float)
    ) {layerSetup [{branchFilters [{boneName=base_hair_01_front}, {boneName=None}]}]}) {layerGroup=None}
   n4 = "Transform (Modify) Bone - Bone: base_brimhat_01_back"(
    ComponentPose=Local To Component(LocalPose=Use cached pose 'Cache_SkydivingState'())
    Alpha=1
    Translation="0,-5,0"
    Rotation="0,-60,20"
    Scale="1,1,1"
   ) {
    rotationMode=BMM_Additive
    rotationSpace=BCS_BoneSpace
    translationMode=BMM_Additive
    translationSpace=BCS_BoneSpace
    alphaBoolBlend {blendOption=Linear}
    alphaScaleBiasClamp {bInterpResult=true}
    boneToModify {boneName=base_brimhat_01_back}
    rotation {roll=20, yaw=-60}
    translation {y=-5}
   }
   n5 = "Transform (Modify) Bone - Bone: base_brimhat_01_back_r"(ComponentPose=@n4, Alpha=1, Translation="0,-5,0", Rotation="0,-5,-40", Scale="1,1,1") {
    rotationMode=BMM_Additive
    rotationSpace=BCS_BoneSpace
    translationMode=BMM_Additive
    translationSpace=BCS_BoneSpace
    alphaBoolBlend {blendOption=Linear}
    alphaScaleBiasClamp {bInterpResult=true}
    boneToModify {boneName=base_brimhat_01_back_r}
    rotation {roll=-40, yaw=-5}
    translation {y=-5}
   }
   n6 = Component To Local(ComponentPose=@n5)
   n7 = Use cached pose 'Cache_SkydivingState'()
   n8 = Get Is Floating And Backpack() -> (Is Floating And Backpack:bool)
   n9 = "Transform (Modify) Bone - Bone: base_brimhat_01_back"(
    ComponentPose=Local To Component(LocalPose=Use cached pose 'Cache_SkydivingState'())
    Alpha=1
    Translation="-0.5,-0.5,1"
    Rotation="0,-15,0"
    Scale="1,1,1"
   ) {
    rotationMode=BMM_Additive
    rotationSpace=BCS_BoneSpace
    translationMode=BMM_Additive
    translationSpace=BCS_BoneSpace
    alphaBoolBlend {blendOption=Linear}
    alphaScaleBiasClamp {bInterpResult=true}
    boneToModify {boneName=base_brimhat_01_back}
    rotation {yaw=-15}
    translation=<-0.5 -0.5 1>
   }
   n10 = "Transform (Modify) Bone - Bone: base_brimhat_01_back_r"(ComponentPose=@n9, Alpha=1, Translation="0,0,0", Rotation="10,5,-10", Scale="1,1,1") {
    rotationMode=BMM_Additive
    rotationSpace=BCS_BoneSpace
    translationMode=BMM_Additive
    translationSpace=BCS_BoneSpace
    alphaBoolBlend {blendOption=Linear}
    alphaScaleBiasClamp {bInterpResult=true}
    boneToModify {boneName=base_brimhat_01_back_r}
    rotation=<rot 10 5 -10>
   }
   Save cached pose 'Cache_SkydivingBackpackCapeState'(Pose=Blend Poses by bool(
     BlendPose_0=Component To Local(ComponentPose=@n10)
     BlendPose_1=Blend Poses by bool(BlendPose_0=@n6, BlendPose_1=@n7, bActiveValue=@n8) {blendType=Linear}
     bActiveValue=Get Is Floating And Cape() -> (Is Floating And Cape:bool)
    ) {blendType=Linear}) {cacheName=Cache_SkydivingBackpackCapeState}
   n11 = Get HatBack() -> (HatBack:double)
   n12 = "Transform (Modify) Bone - Bone: base_hair_01_front_l"(
    ComponentPose=Local To Component(LocalPose=Use cached pose 'Cache_StartPose'())
    Alpha=@n11
    Translation="0,0,0"
    Rotation="0,-2,0"
    Scale="1,1,1"
   ) {
    rotationMode=BMM_Additive
    rotationSpace=BCS_BoneSpace
    alphaBoolBlend {blendOption=Linear}
    alphaScaleBiasClamp {bInterpResult=true}
    boneToModify {boneName=base_hair_01_front_l}
    rotation {yaw=-2}
   }
   n13 = "Transform (Modify) Bone - Bone: base_hair_01_r"(ComponentPose=@n12, Alpha=@n11, Translation="0,0,0", Rotation="-5,0,0", Scale="1,1,1") {
    rotationMode=BMM_Additive
    rotationSpace=BCS_BoneSpace
    alphaBoolBlend {blendOption=Linear}
    alphaScaleBiasClamp {bInterpResult=true}
    boneToModify {boneName=base_hair_01_r}
    rotation {pitch=-5}
   }
   n14 = "Transform (Modify) Bone - Bone: base_hair_01_back_m"(ComponentPose=@n13, Alpha=@n11, Translation="0,0,0", Rotation="20,-20,0", Scale="1,1,1") {
    rotationMode=BMM_Additive
    rotationSpace=BCS_BoneSpace
    translationMode=BMM_Additive
    translationSpace=BCS_BoneSpace
    alphaBoolBlend {blendOption=Linear}
    alphaScaleBiasClamp {bInterpResult=true}
    boneToModify {boneName=base_hair_01_back_m}
    rotation {pitch=20, yaw=-20}
   }
   n15 = "Transform (Modify) Bone - Bone: base_hair_01_back_m_r"(ComponentPose=@n14, Alpha=@n11, Translation="0,-2,0", Rotation="0,-80,0", Scale="1,1,1") {
    rotationMode=BMM_Additive
    rotationSpace=BCS_BoneSpace
    translationMode=BMM_Additive
    translationSpace=BCS_BoneSpace
    alphaBoolBlend {blendOption=Linear}
    alphaScaleBiasClamp {bInterpResult=true}
    boneToModify {boneName=base_hair_01_back_m_r}
    rotation {yaw=-80}
    translation {y=-2}
   }
   n16 = "Transform (Modify) Bone - Bone: base_earring_r"(ComponentPose=@n15, Alpha=@n11, Translation="0,0,0", Rotation="-30,40,-5", Scale="1,1,1") {
    rotationMode=BMM_Additive
    rotationSpace=BCS_BoneSpace
    alphaBoolBlend {blendOption=Linear}
    alphaScaleBiasClamp {bInterpResult=true}
    boneToModify {boneName=base_earring_r}
    rotation=<rot -30 40 -5>
   }
   n17 = "Transform (Modify) Bone - Bone: base_brimhat_01_back"(ComponentPose=@n16, Alpha=@n11, Translation="0,0,0", Rotation="0,-45,60", Scale="1,1,1") {
    rotationMode=BMM_Additive
    rotationSpace=BCS_BoneSpace
    translationMode=BMM_Additive
    translationSpace=BCS_BoneSpace
    alphaBoolBlend {blendOption=Linear}
    alphaScaleBiasClamp {bInterpResult=true}
    boneToModify {boneName=base_brimhat_01_back}
    rotation {roll=60, yaw=-45}
   }
   n18 = Get HatBackRight() -> (HatBackRight:double)
   n19 = "Transform (Modify) Bone - Bone: base_brimhat_01_back_r"(ComponentPose=@n17, Alpha=@n18, Translation="-1.297,-16,5", Rotation="-28.148,-163.529,23.192", Scale="1,1,1") {
    rotationMode=BMM_Additive
    rotationSpace=BCS_BoneSpace
    translationMode=BMM_Additive
    translationSpace=BCS_BoneSpace
    alphaBoolBlend {blendOption=Linear}
    alphaScaleBiasClamp {bInterpResult=true}
    boneToModify {boneName=base_brimhat_01_back_r}
    rotation=<rot -28.148 -163.529 23.192>
    translation=<-1.297 -16 5>
   }
   n20 = "Transform (Modify) Bone - Bone: base_brimhat_01_r"(ComponentPose=@n19, Alpha=@n18, Translation="-0.691,-5.367,-0.346", Rotation="1.327,-0.251,-35.356", Scale="1,1,1") {
    rotationMode=BMM_Additive
    rotationSpace=BCS_BoneSpace
    translationMode=BMM_Additive
    translationSpace=BCS_BoneSpace
    alphaBoolBlend {blendOption=Linear}
    alphaScaleBiasClamp {bInterpResult=true}
    boneToModify {boneName=base_brimhat_01_r}
    rotation=<rot 1.327 -0.2509 -35.3559>
    translation=<-0.691 -5.367 -0.346>
   }
   n21 = Get Is Crouching and Aiming() -> (Is Crouching and Aiming:double)
   n22 = "Transform (Modify) Bone - Bone: base_hair_01_back_m_r"(ComponentPose=@n20, Alpha=@n21, Translation="0,1.5,0", Rotation="5,5,0", Scale="1,1,1") {
    rotationMode=BMM_Additive
    rotationSpace=BCS_BoneSpace
    translationMode=BMM_Additive
    translationSpace=BCS_BoneSpace
    alphaBoolBlend {blendOption=Linear}
    alphaScaleBiasClamp {bInterpResult=true}
    boneToModify {boneName=base_hair_01_back_m_r}
    rotation {pitch=5, yaw=5}
    translation {y=1.5}
   }
   n23 = "Transform (Modify) Bone - Bone: base_brimhat_01_back"(ComponentPose=@n22, Alpha=@n21, Translation="0,0,-2", Rotation="-10,-5,-20", Scale="1,1,1") {
    rotationMode=BMM_Additive
    rotationSpace=BCS_BoneSpace
    translationMode=BMM_Additive
    translationSpace=BCS_BoneSpace
    alphaBoolBlend {blendOption=Linear}
    alphaScaleBiasClamp {bInterpResult=true}
    boneToModify {boneName=base_brimhat_01_back}
    rotation=<rot -10 -5 -20>
    translation {z=-2}
   }
   n24 = "Transform (Modify) Bone - Bone: base_brimhat_01_back_r"(ComponentPose=@n23, Alpha=@n21, Translation="-1,0,-1", Rotation="0,0,-20", Scale="1,1,1") {
    rotationMode=BMM_Additive
    rotationSpace=BCS_BoneSpace
    translationMode=BMM_Additive
    translationSpace=BCS_BoneSpace
    alphaBoolBlend {blendOption=Linear}
    alphaScaleBiasClamp {bInterpResult=true}
    boneToModify {boneName=base_brimhat_01_back_r}
    rotation {roll=-20}
    translation {x=-1, z=-1}
   }
   n25 = RigidBody!(ComponentPose=@n24, Alpha, OverrideWorldGravity:Vector=Get GravityOverride(), ExternalForce) -> (Pose) {
    bClampLinearTranslationLimitToRefPose=true
    bOverrideWorldGravity=true
    overridePhysicsAsset=&/F_MED_ProdigyFire_FaceAcc_FrontHair_Physics
    simulationSpace=BaseBoneSpace
    alphaBoolBlend {blendOption=Linear}
    baseBoneRef {boneName=head}
    componentAppliedLinearAccClamp=<3000 3000 3000>
    componentLinearAccScale=<0.3 0.2 0.2>
    componentLinearVelScale=<0.3 0.2 0.2>
    simSpaceSettings {
     maxAngularAcceleration=1000
     maxAngularVelocity=3
     maxLinearAcceleration=700
     maxLinearVelocity=500
     velocityScaleZ=0
     worldAlpha=0.2
     externalLinearDragV=<0.2 0.2>
    }
   }
   n26 = Anim Dynamics - Bone: dyn_brimhat_01_front(ComponentPose=@n25, Alpha=1, ExternalForce="0,0,0") {
    angularDampingOverride=1
    angularSpringConstant=20000
    bAngularSpring=true
    bOverrideAngularDamping=true
    bUseGravityOverride=true
    gravityScale=0.35
    simulationSpace=RootRelative
    alphaBoolBlend {blendOption=Linear}
    boundBone {boneName=dyn_brimhat_01_front}
    componentLinearAccScale=<1 1 1>
    componentLinearVelScale=<1 1 1>
    gravityOverride {z=-980}
    physicsBodyDefinitions [{
      boundBone {boneName=dyn_brimhat_01_front}
      boxExtents=<20 30 30>
      constraintSetup {angularLimitsMax=<0 0 0>, angularLimitsMin=<0 0 -15>, angularTarget=<1 0 0>}
      localJointOffset=<4 0 0>
     }]
   }
   n27 = Anim Dynamics - Bone: dyn_brimhat_01_l(ComponentPose=@n26, Alpha=1, ExternalForce="0,0,0") {
    angularDampingOverride=1
    angularSpringConstant=20000
    bAngularSpring=true
    bOverrideAngularDamping=true
    bUseGravityOverride=true
    gravityScale=0.35
    simulationSpace=RootRelative
    alphaBoolBlend {blendOption=Linear}
    boundBone {boneName=dyn_brimhat_01_l}
    componentLinearAccScale=<1 1 1>
    componentLinearVelScale=<1 1 1>
    gravityOverride {z=-980}
    physicsBodyDefinitions [{
      boundBone {boneName=dyn_brimhat_01_l}
      boxExtents=<20 30 30>
      constraintSetup {angularLimitsMax=<0 0 0>, angularLimitsMin=<0 0 -15>, angularTarget=<1 0 0>}
      localJointOffset=<4 0 0>
     }]
   }
   n28 = Anim Dynamics - Bone: dyn_brimhat_01_back(ComponentPose=@n27, Alpha=1, ExternalForce="0,0,0") {
    angularDampingOverride=1
    angularSpringConstant=20000
    bAngularSpring=true
    bOverrideAngularDamping=true
    bUseGravityOverride=true
    gravityScale=0.35
    simulationSpace=RootRelative
    alphaBoolBlend {blendOption=Linear}
    boundBone {boneName=dyn_brimhat_01_back}
    componentLinearAccScale=<1 1 1>
    componentLinearVelScale=<1 1 1>
    gravityOverride {z=-980}
    physicsBodyDefinitions [{
      boundBone {boneName=dyn_brimhat_01_back}
      boxExtents=<20 30 30>
      constraintSetup {angularLimitsMax=<0 0 0>, angularLimitsMin=<0 0 -15>, angularTarget=<1 0 0>}
      localJointOffset=<4 0 0>
     }]
   }
   n29 = Anim Dynamics - Bone: dyn_brimhat_02_r(ComponentPose=@n28, Alpha=1, ExternalForce="0,0,0") {
    angularDampingOverride=1
    angularSpringConstant=20000
    bAngularSpring=true
    bOverrideAngularDamping=true
    bUseGravityOverride=true
    gravityScale=0.35
    simulationSpace=RootRelative
    alphaBoolBlend {blendOption=Linear}
    boundBone {boneName=dyn_brimhat_02_r}
    componentLinearAccScale=<1 1 1>
    componentLinearVelScale=<1 1 1>
    gravityOverride {z=-980}
    physicsBodyDefinitions [{
      boundBone {boneName=dyn_brimhat_02_r}
      boxExtents=<20 30 30>
      constraintSetup {angularLimitsMax=<1 1.5 0>, angularLimitsMin=<-1 -1.5 -10>, angularTarget=<1 0 0>}
      localJointOffset=<4 0 0>
     }]
   }
   n30 = Anim Dynamics - Bone: dyn_hair_02_back_m_r_up(ComponentPose=@n29, Alpha=1, ExternalForce="0,0,0") {
    angularDampingOverride=1
    angularSpringConstant=20000
    bAngularSpring=true
    bOverrideAngularDamping=true
    bUseGravityOverride=true
    gravityScale=0.35
    simulationSpace=RootRelative
    alphaBoolBlend {blendOption=Linear}
    boundBone {boneName=dyn_hair_02_back_m_r_up}
    componentLinearAccScale=<1 1 1>
    componentLinearVelScale=<1 1 1>
    gravityOverride {z=-980}
    physicsBodyDefinitions [{
      boundBone {boneName=dyn_hair_02_back_m_r_up}
      boxExtents=<20 30 30>
      constraintSetup {angularLimitsMax=<7.5 5 25>, angularLimitsMin=<-7.5 -5 0>, angularTarget=<1 0 0>}
      localJointOffset=<0 0 0>
     }]
   }
   Save cached pose 'Cache_NormalState'(Pose=Component To Local(ComponentPose=@n30)) {cacheName=Cache_NormalState}
   n31 = Get RBAN Enable() -> (RBAN Enable:double)
   n32 = "Transform (Modify) Bone - Bone: base_brimhat_01_back_r"(
    ComponentPose=Local To Component(LocalPose=Use cached pose 'Cache_StartPose'())
    Alpha=@n31
    Translation="0,0,0"
    Rotation="0,-8,0"
    Scale="1,1,1"
   ) {
    rotationMode=BMM_Additive
    rotationSpace=BCS_BoneSpace
    translationMode=BMM_Additive
    translationSpace=BCS_BoneSpace
    alphaBoolBlend {blendOption=Linear}
    alphaScaleBiasClamp {bInterpResult=true}
    boneToModify {boneName=base_brimhat_01_back_r}
    rotation {yaw=-8}
   }
   n33 = "Transform (Modify) Bone - Bone: base_brimhat_01_back_r"(ComponentPose=@n32, Alpha=Get RBAN Disable() -> (RBAN Disable:double), Translation="0,0,0", Rotation="0,-35,0", Scale="1,1,1") {
    rotationMode=BMM_Additive
    rotationSpace=BCS_BoneSpace
    translationMode=BMM_Additive
    translationSpace=BCS_BoneSpace
    alphaBoolBlend {blendOption=Linear}
    alphaScaleBiasClamp {bInterpResult=true}
    boneToModify {boneName=base_brimhat_01_back_r}
    rotation {yaw=-35}
   }
   n34 = "Transform (Modify) Bone - Bone: base_earring_r"(ComponentPose=@n33, Alpha=@n31, Translation="0.4,-0.6,0.3", Rotation="-35,40,-20", Scale="1,1,1") {
    rotationMode=BMM_Additive
    rotationSpace=BCS_BoneSpace
    translationMode=BMM_Additive
    translationSpace=BCS_BoneSpace
    alphaBoolBlend {blendOption=Linear}
    alphaScaleBiasClamp {bInterpResult=true}
    boneToModify {boneName=base_earring_r}
    rotation=<rot -35 40 -20>
    translation=<0.4 -0.6 0.3>
   }
   n35 = "Transform (Modify) Bone - Bone: base_hair_01_back_l"(ComponentPose=@n34, Alpha=Get Is Parachuting() -> (Is Parachuting:double), Translation="-0.7,0,0", Rotation="0,0,0", Scale="1,1,1") {
    rotationMode=BMM_Additive
    rotationSpace=BCS_BoneSpace
    translationMode=BMM_Additive
    translationSpace=BCS_BoneSpace
    alphaBoolBlend {blendOption=Linear}
    alphaScaleBiasClamp {bInterpResult=true}
    boneToModify {boneName=base_hair_01_back_l}
    translation {x=-0.7}
   }
   n36 = "Transform (Modify) Bone - Bone: base_hair_01_front_l"(ComponentPose=@n35, Alpha=1, Translation="0,0,0", Rotation="0,-2,0", Scale="1,1,1") {
    rotationMode=BMM_Additive
    rotationSpace=BCS_BoneSpace
    alphaBoolBlend {blendOption=Linear}
    alphaScaleBiasClamp {bInterpResult=true}
    boneToModify {boneName=base_hair_01_front_l}
    rotation {yaw=-2}
   }
   n37 = "Transform (Modify) Bone - Bone: base_hair_01_front"(ComponentPose=@n36, Alpha=1, Translation="0,0,0", Rotation="-8,0,0", Scale="1,1,1") {
    rotationMode=BMM_Additive
    rotationSpace=BCS_BoneSpace
    translationMode=BMM_Additive
    translationSpace=BCS_BoneSpace
    alphaBoolBlend {blendOption=Linear}
    alphaScaleBiasClamp {bInterpResult=true}
    boneToModify {boneName=base_hair_01_front}
    rotation {pitch=-8}
   }
   n38 = "Transform (Modify) Bone - Bone: base_hair_01_r"(ComponentPose=@n37, Alpha=1, Translation="0,0,-1", Rotation="-6,0,0", Scale="1,1,1") {
    rotationMode=BMM_Additive
    rotationSpace=BCS_BoneSpace
    translationMode=BMM_Additive
    translationSpace=BCS_BoneSpace
    alphaBoolBlend {blendOption=Linear}
    alphaScaleBiasClamp {bInterpResult=true}
    boneToModify {boneName=base_hair_01_r}
    rotation {pitch=-6}
    translation {z=-1}
   }
   n39 = RigidBody(ComponentPose=@n38, OverrideWorldGravity=Get GravityOverride()) {
    bClampLinearTranslationLimitToRefPose=true
    bOverrideWorldGravity=true
    overridePhysicsAsset=&/F_MED_ProdigyFire_FaceAcc_FrontHairSkyDive_Physics
    simulationSpace=BaseBoneSpace
    alphaBoolBlend {blendOption=Linear}
    baseBoneRef {boneName=head}
    componentAppliedLinearAccClamp=<3000 3000 3000>
    componentLinearAccScale=<0.3 0.2 0.2>
    componentLinearVelScale=<0.3 0.2 0.2>
    simSpaceSettings {
     maxAngularAcceleration=1000
     maxAngularVelocity=3
     maxLinearAcceleration=700
     maxLinearVelocity=500
     velocityScaleZ=0
     worldAlpha=0.2
     externalLinearDragV=<0.2 0.2>
    }
   }
   n40 = Anim Dynamics - Bone: dyn_hair_02_front_l(ComponentPose=@n39, Alpha=1, ExternalForce="0,0,0") {
    angularDampingOverride=1
    angularSpringConstant=20000
    bAngularSpring=true
    bOverrideAngularDamping=true
    bUseGravityOverride=true
    gravityScale=0.35
    simulationSpace=RootRelative
    alphaBoolBlend {blendOption=Linear}
    boundBone {boneName=dyn_hair_02_front_l}
    componentLinearAccScale=<1 1 1>
    componentLinearVelScale=<1 1 1>
    gravityOverride {z=-980}
    physicsBodyDefinitions [{
      boundBone {boneName=dyn_hair_02_front_l}
      boxExtents=<10 70 30>
      constraintSetup {angularLimitsMax=<0 3 0>, angularLimitsMin=<0 -6 -4>, angularTarget=<1 0 0>}
      localJointOffset=<-1 0 0>
     }]
   }
   n41 = Anim Dynamics - Bone: dyn_hair_03_front_l(ComponentPose=@n40, Alpha=1, ExternalForce="0,0,0") {
    angularDampingOverride=1
    angularSpringConstant=20000
    bAngularSpring=true
    bOverrideAngularDamping=true
    bUseGravityOverride=true
    gravityScale=0.35
    simulationSpace=RootRelative
    alphaBoolBlend {blendOption=Linear}
    boundBone {boneName=dyn_hair_03_front_l}
    componentLinearAccScale=<1 1 1>
    componentLinearVelScale=<1 1 1>
    gravityOverride {z=-980}
    physicsBodyDefinitions [{
      boundBone {boneName=dyn_hair_03_front_l}
      boxExtents=<10 70 30>
      constraintSetup {angularLimitsMax=<0 15 0>, angularLimitsMin=<0 -15 -5>, angularTarget=<1 0 0>}
      localJointOffset=<-1 0 0>
     }]
   }
   n42 = Anim Dynamics - Bone: dyn_hair_04_front_l(ComponentPose=@n41, Alpha=1, ExternalForce="0,0,0") {
    angularDampingOverride=1
    angularSpringConstant=20000
    bAngularSpring=true
    bOverrideAngularDamping=true
    bUseGravityOverride=true
    gravityScale=0.35
    simulationSpace=RootRelative
    alphaBoolBlend {blendOption=Linear}
    boundBone {boneName=dyn_hair_04_front_l}
    componentLinearAccScale=<1 1 1>
    componentLinearVelScale=<1 1 1>
    gravityOverride {z=-980}
    physicsBodyDefinitions [{
      boundBone {boneName=dyn_hair_04_front_l}
      boxExtents=<10 70 30>
      constraintSetup {angularLimitsMax=<0 18 0>, angularLimitsMin=<0 -15 -8>, angularTarget=<1 0 0>}
      localJointOffset=<-1 0 0>
     }]
   }
   n43 = Anim Dynamics - Bone: dyn_hair_05_front_l(ComponentPose=@n42, Alpha=1, ExternalForce="0,0,0") {
    angularDampingOverride=1
    angularSpringConstant=20000
    bAngularSpring=true
    bOverrideAngularDamping=true
    bUseGravityOverride=true
    gravityScale=0.35
    simulationSpace=RootRelative
    alphaBoolBlend {blendOption=Linear}
    boundBone {boneName=dyn_hair_05_front_l}
    componentLinearAccScale=<1 1 1>
    componentLinearVelScale=<1 1 1>
    gravityOverride {z=-980}
    physicsBodyDefinitions [{
      boundBone {boneName=dyn_hair_05_front_l}
      boxExtents=<10 70 30>
      constraintSetup {angularLimitsMax=<0 18 0>, angularLimitsMin=<0 -15 -8>, angularTarget=<1 0 0>}
      localJointOffset=<0 0 0>
     }]
   }
   n44 = Anim Dynamics - Bone: dyn_hair_02_front_short_l(ComponentPose=@n43, Alpha=1, ExternalForce="0,0,0") {
    angularDampingOverride=1
    angularSpringConstant=20000
    bAngularSpring=true
    bOverrideAngularDamping=true
    bUseGravityOverride=true
    gravityScale=0.35
    simulationSpace=RootRelative
    alphaBoolBlend {blendOption=Linear}
    boundBone {boneName=dyn_hair_02_front_short_l}
    componentLinearAccScale=<1 1 1>
    componentLinearVelScale=<1 1 1>
    gravityOverride {z=-980}
    physicsBodyDefinitions [{
      boundBone {boneName=dyn_hair_02_front_short_l}
      boxExtents=<10 70 30>
      constraintSetup {angularLimitsMax=<0 25 0>, angularLimitsMin=<0 -15 -8>, angularTarget=<1 0 0>}
      localJointOffset=<-2 0 0>
     }]
   }
   n45 = Anim Dynamics - Bone: dyn_hair_01_front(ComponentPose=@n44, Alpha=1, ExternalForce="0,0,0") {
    angularDampingOverride=1
    angularSpringConstant=20000
    bAngularSpring=true
    bOverrideAngularDamping=true
    bUseGravityOverride=true
    gravityScale=0.35
    simulationSpace=RootRelative
    alphaBoolBlend {blendOption=Linear}
    boundBone {boneName=dyn_hair_01_front}
    componentLinearAccScale=<1 1 1>
    componentLinearVelScale=<1 1 1>
    gravityOverride {z=-980}
    physicsBodyDefinitions [{
      boundBone {boneName=dyn_hair_01_front}
      boxExtents=<10 70 30>
      constraintSetup {angularLimitsMax=<0 20 0>, angularLimitsMin=<0 -10 -3>, angularTarget=<1 0 0>}
      localJointOffset=<-1 0 0>
     }]
   }
   n46 = Anim Dynamics - Bone: dyn_hair_a_02_front(ComponentPose=@n45, Alpha=1, ExternalForce="0,0,0") {
    angularDampingOverride=1
    angularSpringConstant=20000
    bAngularSpring=true
    bOverrideAngularDamping=true
    bUseGravityOverride=true
    gravityScale=0.35
    simulationSpace=RootRelative
    alphaBoolBlend {blendOption=Linear}
    boundBone {boneName=dyn_hair_a_02_front}
    componentLinearAccScale=<1 1 1>
    componentLinearVelScale=<1 1 1>
    gravityOverride {z=-980}
    physicsBodyDefinitions [{
      boundBone {boneName=dyn_hair_a_02_front}
      boxExtents=<10 70 30>
      constraintSetup {angularLimitsMax=<0 20 0>, angularLimitsMin=<0 -10 -4>, angularTarget=<1 0 0>}
      localJointOffset=<-1 0 0>
     }]
   }
   n47 = Anim Dynamics - Bone: dyn_hair_a_03_front(ComponentPose=@n46, Alpha=1, ExternalForce="0,0,0") {
    angularDampingOverride=1
    angularSpringConstant=20000
    bAngularSpring=true
    bOverrideAngularDamping=true
    bUseGravityOverride=true
    gravityScale=0.35
    simulationSpace=RootRelative
    alphaBoolBlend {blendOption=Linear}
    boundBone {boneName=dyn_hair_a_03_front}
    componentLinearAccScale=<1 1 1>
    componentLinearVelScale=<1 1 1>
    gravityOverride {z=-980}
    physicsBodyDefinitions [{
      boundBone {boneName=dyn_hair_a_03_front}
      boxExtents=<10 70 30>
      constraintSetup {angularLimitsMax=<0 25 0>, angularLimitsMin=<0 -15 -5>, angularTarget=<1 0 0>}
      localJointOffset=<0 0 0>
     }]
   }
   n48 = Anim Dynamics - Bone: dyn_hair_b_02_front(ComponentPose=@n47, Alpha=1, ExternalForce="0,0,0") {
    angularDampingOverride=1
    angularSpringConstant=20000
    bAngularSpring=true
    bOverrideAngularDamping=true
    bUseGravityOverride=true
    gravityScale=0.35
    simulationSpace=RootRelative
    alphaBoolBlend {blendOption=Linear}
    boundBone {boneName=dyn_hair_b_02_front}
    componentLinearAccScale=<1 1 1>
    componentLinearVelScale=<1 1 1>
    gravityOverride {z=-980}
    physicsBodyDefinitions [{
      boundBone {boneName=dyn_hair_b_02_front}
      boxExtents=<10 70 30>
      constraintSetup {angularLimitsMax=<0 20 0>, angularLimitsMin=<0 -10 -4>, angularTarget=<1 0 0>}
      localJointOffset=<-2 0 0>
     }]
   }
   n49 = Anim Dynamics - Bone: dyn_hair_b_03_front(ComponentPose=@n48, Alpha=1, ExternalForce="0,0,0") {
    angularDampingOverride=1
    angularSpringConstant=20000
    bAngularSpring=true
    bOverrideAngularDamping=true
    bUseGravityOverride=true
    gravityScale=0.35
    simulationSpace=RootRelative
    alphaBoolBlend {blendOption=Linear}
    boundBone {boneName=dyn_hair_b_03_front}
    componentLinearAccScale=<1 1 1>
    componentLinearVelScale=<1 1 1>
    gravityOverride {z=-980}
    physicsBodyDefinitions [{
      boundBone {boneName=dyn_hair_b_03_front}
      boxExtents=<10 70 30>
      constraintSetup {angularLimitsMax=<0 25 0>, angularLimitsMin=<0 -15 -5>, angularTarget=<1 0 0>}
      localJointOffset=<0 0 0>
     }]
   }
   n50 = Anim Dynamics - Bone: dyn_hair_02_r(ComponentPose=@n49, Alpha=1, ExternalForce="0,0,0") {
    angularDampingOverride=1
    angularSpringConstant=20000
    bAngularSpring=true
    bOverrideAngularDamping=true
    bUseGravityOverride=true
    gravityScale=0.35
    simulationSpace=RootRelative
    alphaBoolBlend {blendOption=Linear}
    boundBone {boneName=dyn_hair_02_r}
    componentLinearAccScale=<1 1 1>
    componentLinearVelScale=<1 1 1>
    gravityOverride {z=-980}
    physicsBodyDefinitions [{
      boundBone {boneName=dyn_hair_02_r}
      boxExtents=<10 70 30>
      constraintSetup {angularLimitsMax=<0 3 0>, angularLimitsMin=<0 -3 -3>, angularTarget=<1 0 0>}
      localJointOffset=<0 0 0>
     }]
   }
   n51 = Anim Dynamics - Bone: dyn_hair_a_03_r(ComponentPose=@n50, Alpha=1, ExternalForce="0,0,0") {
    angularDampingOverride=1
    angularSpringConstant=20000
    bAngularSpring=true
    bOverrideAngularDamping=true
    bUseGravityOverride=true
    gravityScale=0.35
    simulationSpace=RootRelative
    alphaBoolBlend {blendOption=Linear}
    boundBone {boneName=dyn_hair_a_03_r}
    componentLinearAccScale=<1 1 1>
    componentLinearVelScale=<1 1 1>
    gravityOverride {z=-980}
    physicsBodyDefinitions [{
      boundBone {boneName=dyn_hair_a_03_r}
      boxExtents=<10 70 30>
      constraintSetup {angularLimitsMax=<0 2 0>, angularLimitsMin=<0 -10 -5>, angularTarget=<1 0 0>}
      localJointOffset=<0 0 0>
     }]
   }
   n52 = Anim Dynamics - Bone: dyn_hair_a_04_r(ComponentPose=@n51, Alpha=1, ExternalForce="0,0,0") {
    angularDampingOverride=1
    angularSpringConstant=20000
    bAngularSpring=true
    bOverrideAngularDamping=true
    bUseGravityOverride=true
    gravityScale=0.35
    simulationSpace=RootRelative
    alphaBoolBlend {blendOption=Linear}
    boundBone {boneName=dyn_hair_a_04_r}
    componentLinearAccScale=<1 1 1>
    componentLinearVelScale=<1 1 1>
    gravityOverride {z=-980}
    physicsBodyDefinitions [{
      boundBone {boneName=dyn_hair_a_04_r}
      boxExtents=<10 70 30>
      constraintSetup {angularLimitsMax=<0 5 0>, angularLimitsMin=<0 -15 -8>, angularTarget=<1 0 0>}
      localJointOffset=<0 0 0>
     }]
   }
   n53 = Anim Dynamics - Bone: dyn_hair_b_03_r(ComponentPose=@n52, Alpha=1, ExternalForce="0,0,0") {
    angularDampingOverride=1
    angularSpringConstant=20000
    bAngularSpring=true
    bOverrideAngularDamping=true
    bUseGravityOverride=true
    gravityScale=0.35
    simulationSpace=RootRelative
    alphaBoolBlend {blendOption=Linear}
    boundBone {boneName=dyn_hair_b_03_r}
    componentLinearAccScale=<1 1 1>
    componentLinearVelScale=<1 1 1>
    gravityOverride {z=-980}
    physicsBodyDefinitions [{
      boundBone {boneName=dyn_hair_b_03_r}
      boxExtents=<10 70 30>
      constraintSetup {angularLimitsMax=<0 15 0>, angularLimitsMin=<0 0 -20>, angularTarget=<1 0 0>}
      localJointOffset=<0 0 0>
     }]
   }
   n54 = Anim Dynamics - Bone: dyn_hair_b_04_r(ComponentPose=@n53, Alpha=1, ExternalForce="0,0,0") {
    angularDampingOverride=1
    angularSpringConstant=20000
    bAngularSpring=true
    bOverrideAngularDamping=true
    bUseGravityOverride=true
    gravityScale=0.35
    simulationSpace=RootRelative
    alphaBoolBlend {blendOption=Linear}
    boundBone {boneName=dyn_hair_b_04_r}
    componentLinearAccScale=<1 1 1>
    componentLinearVelScale=<1 1 1>
    gravityOverride {z=-980}
    physicsBodyDefinitions [{
      boundBone {boneName=dyn_hair_b_04_r}
      boxExtents=<10 70 30>
      constraintSetup {angularLimitsMax=<0 20 0>, angularLimitsMin=<0 0 -25>, angularTarget=<1 0 0>}
      localJointOffset=<0 0 0>
     }]
   }
   n55 = Anim Dynamics - Bone: dyn_brimhat_01_front(ComponentPose=@n54, Alpha=1, ExternalForce="0,0,0") {
    angularDampingOverride=1
    angularSpringConstant=20000
    bAngularSpring=true
    bOverrideAngularDamping=true
    bUseGravityOverride=true
    gravityScale=0.35
    simulationSpace=RootRelative
    alphaBoolBlend {blendOption=Linear}
    boundBone {boneName=dyn_brimhat_01_front}
    componentLinearAccScale=<1 1 1>
    componentLinearVelScale=<1 1 1>
    gravityOverride {z=-980}
    physicsBodyDefinitions [{
      boundBone {boneName=dyn_brimhat_01_front}
      boxExtents=<20 30 30>
      constraintSetup {angularLimitsMax=<0 0 0>, angularLimitsMin=<0 0 -5>, angularTarget=<1 0 0>}
      localJointOffset=<4 0 0>
     }]
   }
   n56 = Anim Dynamics - Bone: dyn_brimhat_01_l(ComponentPose=@n55, Alpha=1, ExternalForce="0,0,0") {
    angularDampingOverride=1
    angularSpringConstant=20000
    bAngularSpring=true
    bOverrideAngularDamping=true
    bUseGravityOverride=true
    gravityScale=0.35
    simulationSpace=RootRelative
    alphaBoolBlend {blendOption=Linear}
    boundBone {boneName=dyn_brimhat_01_l}
    componentLinearAccScale=<1 1 1>
    componentLinearVelScale=<1 1 1>
    gravityOverride {z=-980}
    physicsBodyDefinitions [{
      boundBone {boneName=dyn_brimhat_01_l}
      boxExtents=<20 30 30>
      constraintSetup {angularLimitsMax=<1 1 0>, angularLimitsMin=<-1 -1 -1>, angularTarget=<1 0 0>}
      localJointOffset=<4 0 0>
     }]
   }
   n57 = Anim Dynamics - Bone: dyn_brimhat_01_back(ComponentPose=@n56, Alpha=1, ExternalForce="0,0,0") {
    angularDampingOverride=1
    angularSpringConstant=20000
    bAngularSpring=true
    bOverrideAngularDamping=true
    bUseGravityOverride=true
    gravityScale=0.35
    simulationSpace=RootRelative
    alphaBoolBlend {blendOption=Linear}
    boundBone {boneName=dyn_brimhat_01_back}
    componentLinearAccScale=<1 1 1>
    componentLinearVelScale=<1 1 1>
    gravityOverride {z=-980}
    physicsBodyDefinitions [{
      boundBone {boneName=dyn_brimhat_01_back}
      boxExtents=<20 30 30>
      constraintSetup {angularLimitsMax=<0 0 0>, angularLimitsMin=<0 0 -5>, angularTarget=<1 0 0>}
      localJointOffset=<4 0 0>
     }]
   }
   n58 = Anim Dynamics - Bone: dyn_hair_02_back_m_r_up(ComponentPose=@n57, Alpha=1, ExternalForce="0,0,0") {
    angularDampingOverride=1
    angularSpringConstant=20000
    bAngularSpring=true
    bOverrideAngularDamping=true
    bUseGravityOverride=true
    gravityScale=0.35
    simulationSpace=RootRelative
    alphaBoolBlend {blendOption=Linear}
    boundBone {boneName=dyn_hair_02_back_m_r_up}
    componentLinearAccScale=<1 1 1>
    componentLinearVelScale=<1 1 1>
    gravityOverride {z=-980}
    physicsBodyDefinitions [{
      boundBone {boneName=dyn_hair_02_back_m_r_up}
      boxExtents=<20 30 30>
      constraintSetup {angularLimitsMax=<7.5 5 15>, angularLimitsMin=<-7.5 -5 -10>, angularTarget=<1 0 0>}
      localJointOffset=<4 0 0>
     }]
   }
   Save cached pose 'Cache_SkydivingState'(Pose=Component To Local(ComponentPose=@n58)) {cacheName=Cache_SkydivingState}
   n59 = Layered blend per bone(BasePose=Copy Pose From Mesh() {bUseAttachedParent=true}, BlendPoses_0=Copy Pose From Mesh(SourceMeshComponent=Get Body_Mesh())) {layerSetup [{branchFilters [
       {boneName=clavicle_l}
       {boneName=clavicle_r}
       {boneName=upperarm_l}
       {boneName=upperarm_r}
       {boneName=upperarm_twist_01_l}
       {boneName=upperarm_twist_01_r}
       {boneName=deform_upperarm_bk_base_l}
       {boneName=deform_upperarm_bk_l}
       {boneName=deform_upperarm_fr_base_l}
       {boneName=deform_upperarm_fr_l}
       {boneName=deform_upperarm_up_base_l}
       {boneName=deform_upperarm_up_l}
       {boneName=deform_upperarm_bk_base_r}
       {boneName=deform_upperarm_bk_r}
       {boneName=deform_upperarm_fr_base_r}
       {boneName=deform_upperarm_fr_r}
       {boneName=deform_upperarm_up_base_r}
       {boneName=deform_upperarm_up_r}
      ]}]}
   Save cached pose 'Cache_StartPose'(Pose=@n59) {cacheName=Cache_StartPose}
  ]}
 EventGraph {nodes [
   Event BlueprintUpdateAnimation()
   > n1 = Sequence!(execute) -> (then_0, then_1, then_2:Exec, then_3:Exec, then_4:Exec, then_5:Exec, then_6:Exec, then_7:Exec, then_8:Exec, then_9:Exec, then_10:Exec)
   Branch(execute=@n1.then_10, Condition=Get bIsInFrontEnd())
   > Set Custom Idle Is Playing Float(Custom Idle Is Playing Float:float=GetCurveValue(self=GetOwnerFrontEndAnimInstance(), CurveName=Frontend_Idle_Bypass)) -> (Output_Get:float)
   Set Is Sky Floating(execute=@n1.then_2, Is Sky Floating:bool=OR Boolean(A=Get bIsSkydiveDiveMode(), B=Get bIsSkydiveFloating())) -> (Output_Get:bool)
   n2 = Gravity Override State Function DEPRECATED(
    execute=@n1.then_0
    CharacterPartAnimInstance=Self-Reference()
    GameTimeInSeconds=GetGameTimeInSeconds()
    GravityOverrideParams=Get Gravity Override Params()
   )
   > Set GravityOverride(GravityOverride=@n2)
   n3 = Get bIsSkydiveDiveMode()
   n4 = Get Is Sky Floating() -> (Is Sky Floating:bool)
   Set Is Floating And Backpack(execute=@n1.then_3, Is Floating And Backpack:bool=AND Boolean(A=OR Boolean(A=@n3, B=@n4), B=Get isWearingBackbling())) -> (Output_Get:bool)
   Set Is Floating And Cape(execute=@n1.then_4, Is Floating And Cape:bool=AND Boolean(A=Get Is Sky Floating(), B=Get Is Wearing Cape() -> (Is Wearing Cape:bool))) -> (Output_Get:bool)
   n5 = "float > float"(A=float * float(A=Get HeadBack() -> (HeadBack:double), B=2) {is=title K2Node_PromotableOperator}, B=1)
   n6 = Get bIsCrouching()
   Set Is Crouching and Aiming(execute=@n1.then_8, Is Crouching and Aiming:double="To Float (Boolean)"(InBool=AND Boolean(A=@n6, B=@n5))) -> (Output_Get:double)
   Set Is Parachuting(execute=@n1.then_5, Is Parachuting:double="To Float (Boolean)"(InBool=Get bIsParachuteOpened())) -> (Output_Get:double)
   n7 = Get Skel Mesh()
   Set HeadSideL(
    execute=@n1.then_9
    HeadSideL:double=GetAlphaFromReferenceTransformDelta(SkeletalMeshComponent=@n7, BoneName=spine_05, BaseName=head, DeltaComponentType=EulerY, InputRange=(X=0,Y=-60), TwistAxis="0,0,1")
   ) -> (Output_Get:double)
   > n8 = Set HeadBack(HeadBack:double=GetAlphaFromReferenceTransformDelta(SkeletalMeshComponent=@n7, BoneName=spine_05, BaseName=head, InputRange=(X=0,Y=90), TwistAxis="0,0,1")) -> (Output_Get:double)
   > n9 = Set HatBack(HatBack:double=@n8) -> (Output_Get:double)
   n10 = float * float(A=Get HeadBack(), B=0.2)
   n11 = Get HeadSideL() -> (HeadSideL:double)
   > Set HatBackRight(HatBackRight:double=float * float(A=@n9, B=float + float(A=@n11, B=@n10))) -> (Output_Get:double)
   Set isSkyDiving(
    execute=@n1.then_1
    isSkyDiving:bool=OR Boolean!(A=Get bIsSkydiveDiveMode(), B=Get bIsSkydiveFloating(), C:bool=Get bIsParachuteOpened()) -> (ReturnValue)
   ) -> (Output_Get:bool)
   n12 = Get bRigidBodyEnabled()
   Set RBAN Disable(execute=@n1.then_7, RBAN Disable:double="To Float (Boolean)"(InBool=NOT Boolean(A=@n12))) -> (Output_Get:double)
   Set RBAN Enable(execute=@n1.then_6, RBAN Enable:double="To Float (Boolean)"(InBool=@n12)) -> (Output_Get:double)
   Event OnCharacterCustomizationLoaded()
   > n13 = Cast To PlayerPawn_Athena(Object=TryGetPawnOwner())
   > Set Body_Mesh(Body_Mesh=GetSkeletalMeshForPartType(self=@n13))
   n14 = CheckForBackbling()
   > Set isWearingBackbling(isWearingBackbling=AND Boolean!(A=@n14, B=NOT Boolean(A=@n14.bIsBackblingACape), C:bool=NOT Boolean(A=@n14.bIsBackblingLong)) -> (ReturnValue))
   > Set Is Wearing Cape(Is Wearing Cape=OR Boolean!(A=@n14.bIsBackblingACape, B=@n14.bIsBackblingLong) -> (ReturnValue))
   Event BlueprintInitializeAnimation()
   > Set Skel Mesh(Skel Mesh=GetOwnerSkeletalMeshComponent())
  ]}
]
variables [
 Body_Mesh {type=SkeletalMeshComponent}
 Custom Idle Is Playing Float {type=number}
 Gravity Override Params {type=GravityOverrideParamsStruct}
 GravityOverride {type=Vector}
 HatBack {type=number}
 HatBackRight {type=number}
 HeadBack {type=number}
 HeadSideL {type=number}
 Is Crouching and Aiming {type=number}
 Is Crouching and Hat Back {type=number}
 Is Floating And Backpack {type=boolean}
 Is Floating And Cape {type=boolean}
 Is Hat Back Right {type=boolean}
 Is Parachuting {type=number}
 Is Sky Floating {type=boolean}
 Is Wearing Cape {type=boolean}
 isSkyDiving {type=boolean}
 isWearingBackbling {type=boolean}
 RBAN Disable {type=number}
 RBAN Enable {type=number}
 Skel Mesh {type=SkeletalMeshComponent}
]
